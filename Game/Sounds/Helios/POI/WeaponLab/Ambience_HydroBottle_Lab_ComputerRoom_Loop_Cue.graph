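# Looping indoor ambience cue for the HydroBottle lab computer room at the Helios Weapon Lab POI.
# A quality-level node picks the quad or stereo version of a looping airy wave.
# Two nested envelopers shape it: an endlessly looping 30-second pitch curve inside a lowered-pitch layer with a gentle 100-second volume swell.
# It plays unspatialized and unattenuated, without reverb send, and its concurrency rule prevents new instances at the limit.
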SoundCue /Game/Sounds/Helios/POI/WeaponLab/Ambience_HydroBottle_Lab_ComputerRoom_Loop_Cue
root=Enveloper_6
nodes [
 {
  id=Enveloper_0
  children [QualityLevel_0]
  properties {
   bLoop=true
   bLoopIndefinitely=true
   loopEnd=30
   pitchCurve {editorCurveData {keys [{interpMode=RCIM_Cubic, time=0}, <at 15.9468 0.9181 RCIM_Cubic>, {interpMode=RCIM_Cubic, time=30}]}}
   volumeCurve {editorCurveData {keys [<at 0 1>]}}
  }
 }
 {
  id=Enveloper_6
  children [Enveloper_0]
  properties {
   bLoop=true
   bLoopIndefinitely=true
   loopEnd=100
   pitchCurve {editorCurveData {keys [<at 0 0.8>]}}
   volumeCurve {editorCurveData {keys [<at 0 1 RCIM_Cubic>,<at 50 1.2 RCIM_Cubic>,<at 100 1 RCIM_Cubic>]}}
  }
 }
 {id=QualityLevel_0, children [WavePlayer_2,WavePlayer_2,WavePlayer_3,WavePlayer_3]}
 {
  id=WavePlayer_2
  properties {bLooping=true, soundWaveAssetPtr=&/Ambience_HydroBottle_Lab_ComputerRoom_Loop_Airy_Quad}
 }
 {
  id=WavePlayer_3
  properties {bLooping=true, soundWaveAssetPtr=&/Ambience_HydroBottle_Lab_ComputerRoom_Loop_Airy_Stereo}
 }
]
properties {
 bOverrideConcurrency=true
 maxDistance=2097152
 soundClassObject=&Fort_Ambient_Indoors_Athena
 volumeMultiplier=0.251
 attenuationOverrides {bAttenuate=false, bEnableReverbSend=false, bSpatialize=false}
 concurrencyOverrides {resolutionRule=PreventNew}
}
